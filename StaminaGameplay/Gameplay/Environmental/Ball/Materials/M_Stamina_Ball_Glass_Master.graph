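# Master material for the Stamina ball's glass: a two-sided masked, anime-shaded shell with a stylized sun glint.
# Facing sign keeps back faces by default, front faces appear only at a hard-stepped sunlight-reflection highlight, and ShadowReplace keeps shadows solid.
# Character PreFX feeds anime shading, then overrides for mask, view-axis-rotated normals, highlight emissive and switchable coin-bob offset feed PostFX.
# A camera-aligned-normal branch is left unconnected.
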Material /StaminaGameplay/Gameplay/Environmental/Ball/Materials/M_Stamina_Ball_Glass_Master
graph {
 properties {
  bUseMaterialAttributes=true
  bUsedWithMeshDeformer=true
  bUsedWithSkeletalMesh=true
  pixelDepthOffsetMode=PDOM_Legacy
  twoSided=true
  usedShadingModels=Default Lit
 }
 nodes [
  n1 = MaterialFunctionCall() {materialFunction=&BaseFN_CharacterFunction_PreFX}
  n2 = GetMaterialAttributes(None=MaterialFunctionCall!(
    Material=@n1.Result
    LitDiffuseOverride
    SSC_AlphaOverride
    Tint Mask
    DF InkLine Visibility Mask
    DF Inkline Injection
    DF Inkline Color Override
    DF Inkline Color Override Mask
    CustomUVs
   ) {materialFunction=/Game/Packages/Fortress_SharedMaterials/Anime_Master_Material/MF_AnimeShading}.Result) {attributeGetTypes [#1, #2]}
  n3 = Multiply(A=TwoSidedSign(), B=ScalarParameter() {defaultValue=-1, group=0SlurpMonsterSummer, parameterName=FlipNormals})
  n4 = CollectionParameter() {
   collection=&FortniteMaterialParameters
   parameterName=SunlightDirectionalVector
  }
  n5 = ComponentMask(None=@n4) {b=true, g=true, r=true}
  n6 = VertexNormalWS()
  n7 = Multiply(A=Multiply(A=DotProduct(A=@n5, B=@n6)) {constB=2}, B=@n6)
  n8 = DotProduct(
   A=Normalize(VectorInput=Subtract(A=@n7, B=@n5))
   B=Multiply(A=MaterialFunctionCall() {materialFunction=&CameraDirectionVector}.Result) {constB=-1}
  )
  n9 = Multiply(A=OneMinus(None=@n3), B=Step(X=Max(A=@n8) {constB=0}) {constX=0.8, constY=0.95})
  n10 = ShadowReplace(Default=Add(A=@n3, B=@n9) {constB=0}, Shadow=Constant() {r=1})
  n11 = VertexNormalWS()
  n12 = RotateAboutAxis(NormalizedRotationAxis=CameraVectorWS(), RotationAngle=Constant() {r=0.5}, PivotPoint=Constant(), Position=@n11)
  n13 = SetMaterialAttributes!(
   MaterialAttributes=@n2.MaterialAttributes
   Opacity Mask=@n10
   Normal=Transform(None=Add(A=@n12, B=@n11)) {transformSourceType=TRANSFORMSOURCE_World, transformType=TRANSFORM_Tangent}
   Emissive Color=LinearInterpolate(
    A=@n2.Emissive Color
    B=VectorParameter() {parameterName=Highlight Color, defaultValue=<rgba 1 0.776 0.3073 1>}.RGB
    Alpha=Saturate(None=@n9)
   )
   World Position Offset=StaticSwitchParameter(
    True=MaterialFunctionCall!(TimeScale, MaxHeight, MinHeight) {materialFunction=/Game/Playgrounds/Items/Props/Materials/MF_Creative_CoinBobble}.WorldPositionOffset
    False=Constant()
   ) {defaultValue=true, parameterName=CreativeMode_CoinBob}
  ) {attributeSetTypes [
    #3
    #1
    #2
    #4
   ]}
  MaterialFunctionCall(AO=@n1.AO, Distance Blend=@n1.AO, Fresnel=@n1.AO, Gmap=@n1.AO, Pre Skinned Local Position=@n1.AO, Material=@n13) {materialFunction=&BaseFN_CharacterFunction_PostFX}
  Saturate(None=@n8)
  n14 = Transform(None=MaterialFunctionCall!(
    WorldPosition=WorldPosition().XYZ
    PivotLocation=MaterialFunctionCall() {materialFunction=&ObjectPivotPoint}.Object Pivot Location
    WS Normals=VertexNormalWS()
    Custom Object Basis 1=Constant3Vector() {constant=<rgba 1 0 0 0>}
    Custom Object Basis 2=Constant3Vector() {constant=<rgba 0 0 -1 0>}
    Custom Object Basis 3=Constant3Vector() {constant=<rgba 0 1 0 0>}
   ) {materialFunction=/Engine/Functions/Engine_MaterialFunctions02/WorldPositionOffset/AlignMeshToTheCamera}.Rotated Normals) {transformSourceType=TRANSFORMSOURCE_World, transformType=TRANSFORM_Tangent}
  SetMaterialAttributes!(
   MaterialAttributes=MaterialFunctionCall(Material=SetMaterialAttributes!(MaterialAttributes=@n1.Result, Opacity Mask, Normal=@n14, Emissive Color) {attributeSetTypes [#3, #1, #2]}) {materialFunction=&MF_AnimeShading}.Result
   Opacity Mask=@n10
   Emissive Color
  ) {attributeSetTypes [#3, #2]}
  Transform(None=VertexNormalWS()) {transformSourceType=TRANSFORMSOURCE_World, transformType=TRANSFORM_Tangent}
  n15 = Constant3Vector() {constant=<rgba 0 0 1 0>}
  Transform(None=Multiply(A=Constant3Vector() {constant=<rgba -1 -1 1 0>}, B=Transform(None=@n15) {transformType=TRANSFORM_Camera})) {transformSourceType=TRANSFORMSOURCE_Camera, transformType=TRANSFORM_Tangent}
 ]
}
