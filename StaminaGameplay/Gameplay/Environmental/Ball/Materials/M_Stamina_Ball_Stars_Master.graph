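# Master material for the stars on the Stamina ball, rendering them anime-shaded and camera-facing.
# MF_AnimeShading output gets rotated normals and a World Position Offset that shrinks the mesh toward its center, aligns it to the camera and pushes it camera-ward.
# A static switch, on by default, adds Creative coin bobbing before the character PostFX function.
# Sun-highlight and LightVector branches exist but do not feed the output.
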Material /StaminaGameplay/Gameplay/Environmental/Ball/Materials/M_Stamina_Ball_Stars_Master
graph {
 properties {
  bTangentSpaceNormal=false
  bUseMaterialAttributes=true
  bUsedWithMeshDeformer=true
  bUsedWithSkeletalMesh=true
  pixelDepthOffsetMode=PDOM_Legacy
  usedShadingModels=Default Lit
 }
 nodes [
  Divide(A=Multiply(A=MaterialFunctionCall() {materialFunction=&LocalPosition}.Local Position)) {constB=150}
  n1 = MaterialFunctionCall() {materialFunction=&BaseFN_CharacterFunction_PreFX}
  n2 = GetMaterialAttributes(None=MaterialFunctionCall!(
    Material=@n1.Result
    LitDiffuseOverride
    SSC_AlphaOverride
    Tint Mask
    DF InkLine Visibility Mask
    DF Inkline Injection
    DF Inkline Color Override
    DF Inkline Color Override Mask
    CustomUVs
   ) {materialFunction=/Game/Packages/Fortress_SharedMaterials/Anime_Master_Material/MF_AnimeShading}.Result) {attributeGetTypes [#1, #2]}
  n3 = CollectionParameter() {
   collection=&FortniteMaterialParameters
   parameterName=SunlightDirectionalVector
  }
  n4 = ComponentMask(None=@n3) {b=true, g=true, r=true}
  n5 = MaterialFunctionCall() {materialFunction=&ObjectPivotPoint}
  n6 = Subtract(A=WorldPosition().XYZ, B=@n5.Object Pivot Location)
  n7 = Add(
   A=Multiply(A=Normalize(VectorInput=@n6), B=ScalarParameter() {defaultValue=7, parameterName=FakeDistance}) {constB=7}
   B=@n5.Object Pivot Location
  )
  n8 = Multiply(A=Multiply(A=DotProduct(A=@n4, B=@n7)) {constB=2}, B=VertexNormalWS())
  n9 = DotProduct(
   A=Normalize(VectorInput=Subtract(A=@n8, B=@n4))
   B=Multiply(A=MaterialFunctionCall() {materialFunction=&CameraDirectionVector}.Result) {constB=-1}
  )
  LinearInterpolate(A=@n2.Emissive Color, Alpha=Step(X=Max(A=@n9) {constB=0}) {constX=0.8, constY=0.95})
  n10 = WorldPosition() {worldPositionShaderOffset=WPT_ExcludeAllShaderOffsets}
  n11 = Subtract(A=ObjectPositionWS(), B=@n10.XYZ)
  n12 = ScalarParameter() {defaultValue=0.1, parameterName=Star shrink amount}
  n13 = MaterialFunctionCall!(
   WorldPosition=Add(A=@n10.XYZ, B=Multiply(A=@n11, B=@n12))
   PivotLocation
   WS Normals=Transform(None=@n2.Normal)
   Custom Object Basis 1=Constant3Vector() {constant=<rgba 1 0 0 0>}
   Custom Object Basis 2=Constant3Vector() {constant=<rgba 0 0 -1 0>}
   Custom Object Basis 3=Constant3Vector() {constant=<rgba 0 1 0 0>}
  ) {materialFunction=/Engine/Functions/Engine_MaterialFunctions02/WorldPositionOffset/AlignMeshToTheCamera}
  n14 = Subtract(A=CameraPositionWS(), B=ObjectPositionWS())
  n15 = Add(
   A=@n13.World Position Offset
   B=Multiply(A=Normalize(VectorInput=@n14), B=ScalarParameter() {defaultValue=7, parameterName=Star push amount towards camera}) {constB=5}
  )
  n16 = StaticSwitchParameter(
   True=MaterialFunctionCall!(TimeScale, MaxHeight, MinHeight) {materialFunction=/Game/Playgrounds/Items/Props/Materials/MF_Creative_CoinBobble}.WorldPositionOffset
   False=Constant()
  ) {defaultValue=true, parameterName=CreativeMode_CoinBob}
  MaterialFunctionCall(
   AO=@n1.AO
   Distance Blend=@n1.AO
   Fresnel=@n1.AO
   Gmap=@n1.AO
   Pre Skinned Local Position=@n1.AO
   Material=SetMaterialAttributes!(MaterialAttributes=@n2.MaterialAttributes, World Position Offset=Add(A=@n15, B=@n16), Normal=@n13.Rotated Normals, Emissive Color) {attributeSetTypes [#3, #1, #2]}
  ) {materialFunction=&BaseFN_CharacterFunction_PostFX}
  n17 = CollectionParameter() {
   collection=&FortniteMaterialParameters
   parameterName=SunlightDirectionalVector
  }
  n18 = StaticSwitchParameter(True=Custom() {code="return View.DirectionalLightDirection;", description=DirectionalLightDirection, inputs []}, False=@n17) {
   desc=use the Light vector of the first directional light in the scene instead of the Sun Vector.
   group=Toon_Shading
   parameterName=UseDirectionalLightFunction
   sortPriority=0
  }
  n19 = MaterialFunctionCall(Float3=@n18) {materialFunction=&BreakOutFloat3Components}
  n20 = Normalize(VectorInput=MaterialFunctionCall(X=@n19.R, Y=@n19.R, Z=Constant()) {materialFunction=&MakeFloat3}.Result) {desc=need to transform it to preskinning}
  NamedRerouteDeclaration(None=ComponentMask(None=StaticSwitchParameter(True=@n20, False=@n18) {desc=Lock the Z axis of the Light Vector, group=Toon_Shading, parameterName=UseLockLightZVector, sortPriority=0}) {b=true, g=true, r=true}) {name=LightVector, nodeColor=<rgba 0 1 0.8634 1>}
  Saturate(None=@n9)
  n21 = CollectionParameter() {
   collection=&FortniteMaterialParameters
   parameterName=SunlightDirectionalVector
  }
  Subtract(
   A=@n21
   B=MaterialFunctionCall() {materialFunction=&ObjectPivotPoint}.Object Pivot Location
  )
  VertexInterpolator(VS=PreSkinnedNormal())
  VertexNormalWS()
  VertexNormalWS()
 ]
}
